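# Root conversation database for the ElbowChat NPC, who speaks as Mephisto and greets players at the start of a conversation.
# One entry point feeds a greeting Speech node with a default line and contextual lines for first-ever meetings, match conversation count and specific player outfits.
# First-meeting and outfit lines take their priority from named curve rows.
# The greeting branches through link nodes to the shared NPCLibrary Quests entry and ElbowChat's Services and Dialogue entries.
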ConversationDatabase /ElbowChatNPC/NPC/ElbowChat/NPCLibrary_Conversation_ElbowChat
entry_points [{tag=Conversation.Entry.ElbowChat, destinations [#1]}]
nodes [
 ConversationEntryPointNode_7 {
  id=#1
  class=ConversationEntryPointNode
  outputs [#2]
  properties {entryTag {tagName=Conversation.Entry.ElbowChat}}
 }
 Speech_C_4 {
  id=#2
  class=Speech_C
  outputs [#3, #4, #5]
  properties {
   nodeName=""
   generalConfig {
    defaultMessage=What would you give to receive my gift?
    contextualMessages [
     {
      message="Scrape and bow, mortal. You gaze upon Mephisto! "
      contextRequirements [{requirement {class=FortControllerRequirement_IsFirstEverConversationWithNPC}}]
      priority {curve {rowName=Tandem.Speech.Priority.FirstEverInteraction}}
     }
     {
      message=What would you give to receive my gift?
      contextRequirements [{requirement {class=FortControllerRequirement_CheckNPCMatchConversationCount, properties {nPCUniqueIdTag {tagName=AISpawnerData.Type.Tandem.ElbowChat}}}}]
      priority {curve {rowName=None}}
     }
     {
      message=My influence is inescapable.
      contextRequirements [{requirement {class=FortControllerRequirement_CheckNPCMatchConversationCount, properties {nPCUniqueIdTag {tagName=AISpawnerData.Type.Tandem.ElbowChat}}}}]
      priority {curve {rowName=None}}
     }
     {
      message=You're in Mephisto's domain. You serve at his pleasure.
      contextRequirements [{requirement {class=FortControllerRequirement_CheckNPCMatchConversationCount, properties {nPCUniqueIdTag {tagName=AISpawnerData.Type.Tandem.ElbowChat}}}}]
      priority {curve {rowName=None}}
     }
     {
      message=Your mother's screams echo across my realm.
      contextRequirements [{requirement {
         class=FortControllerRequirement_HasCID
         properties {allowedCIds [
           /BRCosmetics/Athena/Items/Cosmetics/Characters/CID_846_Athena_Commando_M_HightowerDate
           /BRCosmetics/Athena/Items/Cosmetics/Characters/Character_AnglePatch_Lodge
          ]}
        }}]
      priority {curve {rowName=Tandem.Speech.Priority.OutfitReaction}}
     }
     {
      message=A sorceror supreme is no match for Mephisto.
      contextRequirements [{requirement {
         class=FortControllerRequirement_HasCID
         properties {allowedCIds [/BRCosmetics/Athena/Items/Cosmetics/Characters/CID_A_367_Athena_Commando_M_Mystic]}
        }}]
      priority {curve {rowName=Tandem.Speech.Priority.OutfitReaction}}
     }
     {
      message=Your inner goodness radiates once more. Repulsive.
      contextRequirements [{requirement {
         class=FortControllerRequirement_HasCID
         properties {allowedCIds [
           /BRCosmetics/Athena/Items/Cosmetics/Characters/CID_845_Athena_Commando_M_HightowerTapas
           /BRCosmetics/Athena/Items/Cosmetics/Characters/CID_A_439_Athena_Commando_M_Trifle
          ]}
        }}]
      priority {curve {rowName=Tandem.Speech.Priority.OutfitReaction}}
     }
     {
      message=Wanda. How are your boys?
      contextRequirements [{requirement {
         class=FortControllerRequirement_HasCID
         properties {allowedCIds [/BRCosmetics/Athena/Items/Cosmetics/Characters/CID_A_399_Athena_Commando_F_Ultralight]}
        }}]
      priority {curve {rowName=Tandem.Speech.Priority.OutfitReaction}}
     }
     {
      message="I will not lose to another Black Panther. "
      contextRequirements [{requirement {class=FortControllerRequirement_HasCID, properties {allowedCIds [/BRCosmetics/Athena/Items/Cosmetics/Characters/Character_SnapFreeze_Hunt]}}}]
      priority {curve {rowName=Tandem.Speech.Priority.OutfitReaction}}
     }
     {
      message="Believe me, kitten: there is no greater joy than defeating Victor Von Doom."
      contextRequirements [{requirement {
         class=FortControllerRequirement_HasCID
         properties {allowedCIds [
           &Character_VoidRedemption_Rebel
           &Character_BrightDisk
          ]}
        }}]
      priority {curve {rowName=Tandem.Speech.Priority.OutfitReaction}}
     }
     {
      message=Are you happy with the bargain you've made?
      contextRequirements [{requirement {
         class=FortControllerRequirement_HasCID
         properties {allowedCIds [/BRCosmetics/Athena/Items/Cosmetics/Characters/CID_924_Athena_Commando_M_Embers]}
        }}]
      priority {curve {rowName=Tandem.Speech.Priority.OutfitReaction}}
     }
     {
      message=Our bargain has charred you...
      contextRequirements [{requirement {class=FortControllerRequirement_HasCID, properties {allowedCIds [/BRCosmetics/Athena/Items/Cosmetics/Characters/Character_FreeDrive]}}}]
      priority {curve {rowName=Tandem.Speech.Priority.OutfitReaction}}
     }
    ]
   }
  }
 }
 ConversationLinkNode_2 {id=#5, class=ConversationLinkNode, properties {remoteEntryTag {tagName=Conversation.Entry.ElbowChat.Dialogue}}}
 ConversationLinkNode_0 {id=#4, class=ConversationLinkNode, properties {remoteEntryTag {tagName=Conversation.Entry.ElbowChat.Services}}}
 ConversationLinkNode_1 {id=#3, class=ConversationLinkNode, properties {remoteEntryTag {tagName=Conversation.Entry.NPCLibrary.Quests}}}
]
