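# Conversation database for the ElbowChat NPC's "Ask about..." dialogue, in which the NPC gloats over Doctor Doom and Cynthia Von Doom.
# The tagged entry point opens a greeting Speech node that branches by choice sub-nodes into the two topics or a Back node.
# Each topic chains three Speech lines, with only the first carrying a choice label.
# Both chains converge on one shared ReturnToStart node labeled BACK.
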ConversationDatabase /ElbowChatNPC/NPC/ElbowChat/NPCLibrary_Conversation_ElbowChat_Dialogue
entry_points [{tag=Conversation.Entry.ElbowChat.Dialogue, destinations [#1]}]
nodes [
 ConversationEntryPointNode_0 {
  id=#1
  class=ConversationEntryPointNode
  outputs [#2]
  properties {entryTag {tagName=Conversation.Entry.ElbowChat.Dialogue}}
 }
 Speech_C_0 {
  id=#2
  class=Speech_C
  outputs [#3, #4, #5]
  properties {
   nodeName=""
   softTaskIcon=&T_UI_NPC_question
   generalConfig {defaultMessage="Hmm. Why not? I'll entertain your petty queries. "}
  }
  sub_nodes [ConversationChoiceNode_5 {class=ConversationChoiceNode, properties {defaultChoiceDisplayText=Ask about...}}]
 }
 Speech_C_1 {
  id=#5
  class=Speech_C
  outputs [#6]
  properties {
   nodeName=""
   softTaskIcon=&T_UI_NPC_question
   generalConfig {defaultMessage=Latveria's strong-handed master will dedicate his life to power. Control. Perfection.}
  }
  sub_nodes [ConversationChoiceNode_15 {class=ConversationChoiceNode, properties {defaultChoiceDisplayText=Doctor Doom}}]
 }
 Speech_C_2 {
  id=#6
  class=Speech_C
  outputs [#7]
  properties {
   nodeName=""
   softTaskIcon=&T_UI_NPC_question
   generalConfig {defaultMessage=But he'll never get what he truly wants.}
  }
 }
 Speech_C_3 {
  id=#7
  class=Speech_C
  outputs [#8]
  properties {
   nodeName=""
   softTaskIcon=&T_UI_NPC_question
   generalConfig {defaultMessage="How lucky I am to witness his eternal torment. It is... delicious!"}
  }
 }
 ReturnToStart_C_23 {
  id=#8
  class=ReturnToStart_C
  properties {softTaskIcon=&T_UI_NPC_Back}
  sub_nodes [ConversationChoiceNode_32 {class=ConversationChoiceNode, properties {defaultChoiceDisplayText=BACK}}]
 }
 Back_C_0 {
  id=#4
  class=Back_C
  sub_nodes [ConversationChoiceNode_7 {class=ConversationChoiceNode, properties {defaultChoiceDisplayText=BACK}}]
 }
 Speech_C_7 {
  id=#3
  class=Speech_C
  outputs [#9]
  properties {
   nodeName=""
   softTaskIcon=&T_UI_NPC_question
   generalConfig {defaultMessage=Cynthia bargained with me for power. And she got it.}
  }
  sub_nodes [ConversationChoiceNode_0 {class=ConversationChoiceNode, properties {defaultChoiceDisplayText=Cynthia Von Doom}}]
 }
 Speech_C_15 {
  id=#9
  class=Speech_C
  outputs [#10]
  properties {
   nodeName=""
   softTaskIcon=&T_UI_NPC_question
   generalConfig {defaultMessage="She did not ask for *control* over that power... "}
  }
 }
 Speech_C_14 {
  id=#10
  class=Speech_C
  outputs [#8]
  properties {
   nodeName=""
   softTaskIcon=&T_UI_NPC_question
   generalConfig {defaultMessage=That failure is the origin of her doom.}
  }
 }
]
