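# Conversation database for the ElbowChat NPC's service menu, offering "Mephisto's Gift" deals paid in player health.
# The entry branches to two service-gated speech previews, weapon and team reboot, each leading to Accept or Back.
# Weapon grants loot-tier items with ammo; reboot revives the team without cards, requiring a non-solo playlist and eliminated teammates.
# Both check and remove health using shared curve-table costs, leaving failed offers visible with a reason.
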ConversationDatabase /ElbowChatNPC/NPC/ElbowChat/NPCLibrary_Conversation_ElbowChat_Services
entry_points [{tag=Conversation.Entry.ElbowChat.Services, destinations [#1]}]
nodes [
 ConversationEntryPointNode_0 {
  id=#1
  class=ConversationEntryPointNode
  outputs [#2, #3]
  properties {entryTag {tagName=Conversation.Entry.ElbowChat.Services}}
 }
 Speech_C_1 {
  id=#3
  class=Speech_C
  outputs [#4, #5]
  properties {
   bUsePerNPC_ContextualMessageLookup=true
   softTaskIcon=/Game/Athena/HUD/Interaction/T_UI_NPC_ReviveTeammate
   generalConfig {defaultMessage=No refunds.}
   speakerEntryTagToConfig {"(TagName=\"Conversation.Entry.ElbowChat\")" {defaultMessage=Raise your entire team... for a price.}}
  }
  sub_nodes [
   FortConversationRequirement_HasService_0 {class=FortConversationRequirement_HasService, properties {serviceTag {tagName=Tandem.Service.OfferReboot}}}
   FortConversationChoiceWithPreviewNode_0 {
    class=FortConversationChoiceWithPreviewNode
    properties {defaultChoiceDisplayText=MEPHISTO'S GIFT, choiceTags {gameplayTags [{tagName=Conversation.Choice.CustomSortPriority.Lowest}]}}
   }
  ]
 }
 OfferTeamReboot_C_0 {
  id=#5
  class=OfferTeamReboot_C
  outputs []
  properties {
   effectRecipientConfigs [{
     recipient=Player
     effects [
      {
       class=UFortControllerEffect_RemoveHealth_C
       properties {
        bSetMagnitudeFromParameter=false
        healthCostCurveTable {curveTable=/ElbowChatNPC/DataTables/ElbowChatServiceData, rowName=OfferTeamReboot.Cost}
       }
      }
      {class=UFortControllerEffect_RebootTeamNoCards_C}
     ]
    }]
   requirements [
    {
     cannotUseReasonParameter=requiredPlaylist
     failureNodeBehaviour=FailedButVisible
     participantId {tagName=Conversation.Participant.Player}
     requirement {
      class=UFortControllerRequirement_Playlist_C
      properties {
       bInverseCondition=true
       playlistContextTags {gameplayTags [{tagName=Athena.Playlist.Solo}, {tagName=Athena.Playlist.NoBuildBR.Solo}, {tagName=Athena.Playlist.Bot.Solo}]}
      }
     }
    }
    {
     cannotUseReasonParameter=eliminatedTeam
     failureNodeBehaviour=FailedButVisible
     participantId {tagName=Conversation.Participant.Player}
     requirement {class=UFortControllerRequirement_EliminatedTeam_C}
    }
    {
     cannotUseReasonParameter=notEnoughHealth
     failureNodeBehaviour=FailedButVisible
     participantId {tagName=Conversation.Participant.Player}
     requirement {
      class=UFortControllerRequirement_HasHealth_C
      properties {healthCostTable {curveTable=&ElbowChatServiceData, rowName=OfferTeamReboot.Cost}}
     }
    }
   ]
  }
  sub_nodes [ConversationChoiceNode_0 {class=ConversationChoiceNode, properties {defaultChoiceDisplayText=ACCEPT, choiceTags {gameplayTags [{tagName=Tandem.Service.OfferReboot}]}}}]
 }
 Back_C_1 {
  id=#4
  class=Back_C
  sub_nodes [ConversationChoiceNode_2 {class=ConversationChoiceNode, properties {defaultChoiceDisplayText=BACK}}]
 }
 Speech_C_0 {
  id=#2
  class=Speech_C
  outputs [#6, #7]
  properties {
   bUsePerNPC_ContextualMessageLookup=true
   softTaskIcon=/Game/Athena/HUD/Interaction/T_UI_NPC_ReceiveRandomWeapon
   generalConfig {defaultMessage=No refunds.}
   speakerEntryTagToConfig {"(TagName=\"Conversation.Entry.ElbowChat\")" {defaultMessage="I don't give lightly, you know."}}
  }
  sub_nodes [
   FortConversationChoiceWithPreviewNode_19 {
    class=FortConversationChoiceWithPreviewNode
    properties {defaultChoiceDisplayText=MEPHISTO'S GIFT, choiceTags {gameplayTags [{tagName=Conversation.Choice.CustomSortPriority.Highest}]}}
   }
   FortConversationRequirement_HasService_1 {class=FortConversationRequirement_HasService, properties {serviceTag {tagName=Tandem.Service.OfferWeapon}}}
  ]
 }
 OfferWeapon_C_2 {
  id=#7
  class=OfferWeapon_C
  outputs []
  properties {
   effectRecipientConfigs [{
     recipient=Player
     effects [
      {
       class=UFortControllerEffect_RemoveHealth_C
       properties {bSetMagnitudeFromParameter=false, healthCostCurveTable {curveTable=&ElbowChatServiceData, rowName=OfferWeapon.Cost}}
      }
      {
       class=UFortControllerEffect_GrantItem_C
       properties {
        tier Group Name=LTG_NPC_Service_OfferWeapon
        ammoQuantityMap {
         /Script/FortniteGame.FortAmmoItemDefinition'/Game/Athena/Items/Ammo/AmmoDataRockets.AmmoDataRockets'=8
         &FortAmmoItemDefinition'/Game/Athena/Items/Ammo/AthenaAmmoDataBulletsLight.AthenaAmmoDataBulletsLight'=26
         &FortAmmoItemDefinition'/Game/Athena/Items/Ammo/AthenaAmmoDataBulletsMedium.AthenaAmmoDataBulletsMedium'=31
         &FortAmmoItemDefinition'/Game/Athena/Items/Ammo/AthenaAmmoDataShells.AthenaAmmoDataShells'=12
        }
       }
      }
     ]
    }]
   requirements [{
     cannotUseReasonParameter=notEnoughHealth
     failureNodeBehaviour=FailedButVisible
     participantId {tagName=Conversation.Participant.Player}
     requirement {
      class=UFortControllerRequirement_HasHealth_C
      properties {healthCostTable {curveTable=&ElbowChatServiceData, rowName=OfferWeapon.Cost}}
     }
    }]
  }
  sub_nodes [ConversationChoiceNode_5 {class=ConversationChoiceNode, properties {defaultChoiceDisplayText=ACCEPT, choiceTags {gameplayTags [{tagName=Tandem.Service.OfferWeapon}]}}}]
 }
 Back_C_13 {
  id=#6
  class=Back_C
  sub_nodes [ConversationChoiceNode_1 {class=ConversationChoiceNode, properties {defaultChoiceDisplayText=BACK}}]
 }
]
